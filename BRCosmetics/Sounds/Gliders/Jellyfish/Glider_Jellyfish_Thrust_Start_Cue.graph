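# Sound cue for the thrust-start sound of the Jellyfish glider in Battle Royale.
# A Random node picks one of three equally weighted thrust waves, and a Modulator above it slightly randomizes the volume.
# Playback is quieted by a low volume multiplier and uses shared glider thrust attenuation, a glider landing sound class and an Athena glider thrust concurrency set.
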SoundCue /BRCosmetics/Sounds/Gliders/Jellyfish/Glider_Jellyfish_Thrust_Start_Cue
root=Modulator_0
nodes [
 {id=Modulator_0, children [Random_1], properties {volumeMax=1, volumeMin=0.85}}
 {id=Random_1, children [WavePlayer_1,WavePlayer_5,WavePlayer_6], properties {weights [1,1,1]}}
 {id=WavePlayer_1, properties {soundWaveAssetPtr=&/SFX_Glider_Jellyfish_Thrust_01}}
 {id=WavePlayer_5, properties {soundWaveAssetPtr=&/SFX_Glider_Jellyfish_Thrust_02}}
 {id=WavePlayer_6, properties {soundWaveAssetPtr=&/SFX_Glider_Jellyfish_Thrust_03}}
]
properties {
 attenuationSettings=&Fort_Player_Glider_Thrust_Athena_Att
 maxDistance=12500
 soundClassObject=&Fort_Gliders_Land
 volumeMultiplier=0.15
 concurrencySet [&Athena_Glider_Thrust_Conc]
}
